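; Sound spline component that drives one owner-attached audio component from the player's position relative to a spline, for stereo-aware sound along a path.
; It samples points left and right of player 0, finds each one's closest spline location, and computes distance and angle from player forward.
; These feed sound parameters, with audio fading in or out by range.
; A self-rescheduling timer replaces Tick, scaling its interval with the nearer distance.
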(event Custom|Start%20Updates
  (CallFunction|ScheduleUpdate 0.1))

(event Custom|Schedule%20Update
  (params (Time%20Until%20Next%20Update :type Float%20%28single-precision%29 :default 0.1))
  (bind _returnvalue (Utilities|Time|SetTimerbyEvent (EventDispatchers|CreateEvent :!event Do%20Work) Time%20Until%20Next%20Update false true))
  (Variables|Internal|SetUpdate%20Timer _returnvalue))

(event Custom|Stop%20Updates
  (Utilities|Time|ClearandInvalidateTimerbyHandle (Variables|Internal|GetUpdate%20Timer)))

(event EventEndPlay
  (params (EndPlayReason :type EEndPlayReason%20Enum :default "Destroyed"))
  (SoundSpline|StopSplineAudio)
  (CallFunction|StopUpdates))

(fn ComputeDistance
  (bind _sound_spline_settings (Variables|SoundSpline|GetSound%20Spline%20Settings))
  (bind (_calculate2ddistance_18_7ed6164344da21eaf65ddbb8511777fc _samplewidth_7_fcc21aa84200281c7a951d9610a76c8d) (Utilities|Struct|BreakSoundSplineSettings _sound_spline_settings))
  (bind _returnvalue (Game|GetPlayerPawn 0))
  (bind _player_location (LocalVariables|GetPlayer%20Location))
  (bind _z (.z _player_location))
  (bind _left_spline_location (LocalVariables|GetLeft%20Spline%20Location))
  (bind _returnvalue_2 (Math|Conversions|ToVector2D(Vector) _left_spline_location))
  (bind _returnvalue_3 (Math|Conversions|ToVector(Vector2D) _returnvalue_2 _z))
  (bind _time_until_next_update (SoundSpline|GetTimeUntilNextUpdate))
  (bind _right_spline_location (LocalVariables|GetRight%20Spline%20Location))
  (bind _returnvalue_4 (Math|Conversions|ToVector2D(Vector) _right_spline_location))
  (bind _returnvalue_5 (Math|Conversions|ToVector(Vector2D) _returnvalue_4 _z))
  (bind _output_get (LocalVariables|SetSpline%20Component (Utilities|Struct|BreakSoundSplineData (Variables|Internal|GetSound%20Spline%20Data))))
  (Utilities|IsValid _output_get
    (:Is%20Valid
      (Spline|SetDrawDebug _output_get (Variables|SoundSpline|GetDebug))
      (Utilities|FlowControl|Sequence
        (:then_0
          (LocalVariables|SetCalculate%202D%20Distance _calculate2ddistance_18_7ed6164344da21eaf65ddbb8511777fc)
          (LocalVariables|SetSample%20Width _samplewidth_7_fcc21aa84200281c7a951d9610a76c8d)
          (LocalVariables|SetPlayer%20Location (Transformation|GetActorLocation _returnvalue))
          (LocalVariables|SetPlayer%20Right%20Vector (Transformation|GetActorRightVector _returnvalue))
          (LocalVariables|SetPlayer%20Forward%20Vector (Math|Conversions|ToVector(Vector2D) (Math|Conversions|ToVector2D(Vector) (Transformation|GetActorForwardVector _returnvalue)))))
        (:then_1
          (Utilities|FlowControl|Sequence
            (:then_0
              (LocalVariables|SetLeft%20Sample%20Location (+ (LocalVariables|GetPlayer%20Location) (Utilities|Operators|Multiply (LocalVariables|GetSample%20Width) -1.0 (LocalVariables|GetPlayer%20Right%20Vector))))
              (LocalVariables|SetRight%20Sample%20Location (+ (LocalVariables|GetPlayer%20Location) (Utilities|Operators|Multiply (LocalVariables|GetSample%20Width) 1.0 (LocalVariables|GetPlayer%20Right%20Vector)))))
            (:then_1
              (Utilities|FlowControl|Sequence
                (:then_0
                  (LocalVariables|SetLeft%20Spline%20Location (Spline|FindLocationClosesttoWorldLocation (LocalVariables|GetSpline%20Component) (LocalVariables|GetLeft%20Sample%20Location) "World"))
                  (LocalVariables|SetRight%20Spline%20Location (Spline|FindLocationClosesttoWorldLocation (LocalVariables|GetSpline%20Component) (LocalVariables|GetRight%20Sample%20Location) "World")))
                (:then_1
                  (Utilities|FlowControl|Sequence
                    (:then_0
                      (bind _structout (Utilities|Struct|SetmembersinSoundSplineData (Variables|Internal|GetSound%20Spline%20Data) (Math|Vector|VectorLength (Math|Vector|SelectVector (Math|Conversions|ToVector(Vector2D) (- (Math|Conversions|ToVector2D(Vector) (LocalVariables|GetLeft%20Spline%20Location)) (Math|Conversions|ToVector2D(Vector) (LocalVariables|GetLeft%20Sample%20Location)))) (LocalVariables|GetLeft%20Spline%20Location) (LocalVariables|GetCalculate%202D%20Distance))) (Math|Trig|Acos(Degrees) (Math|Vector|DotProduct (Math|Vector|Normalize (Math|Vector|SelectVector (Math|Conversions|ToVector(Vector2D) (- (Math|Conversions|ToVector2D(Vector) (LocalVariables|GetLeft%20Spline%20Location)) (Math|Conversions|ToVector2D(Vector) (LocalVariables|GetLeft%20Sample%20Location)))) (LocalVariables|GetLeft%20Spline%20Location) (LocalVariables|GetCalculate%202D%20Distance))) (LocalVariables|GetPlayer%20Forward%20Vector)))))
                      (Variables|Internal|SetSound%20Spline%20Data _structout))
                    (:then_1
                      (Utilities|FlowControl|Sequence
                        (:then_0
                          (bind _structout_2 (Utilities|Struct|SetmembersinSoundSplineData (Variables|Internal|GetSound%20Spline%20Data) (Math|Vector|VectorLength (Math|Vector|SelectVector (Math|Conversions|ToVector(Vector2D) (- (Math|Conversions|ToVector2D(Vector) (LocalVariables|GetRight%20Spline%20Location)) (Math|Conversions|ToVector2D(Vector) (LocalVariables|GetRight%20Sample%20Location)))) (LocalVariables|GetRight%20Spline%20Location) (LocalVariables|GetCalculate%202D%20Distance))) (Math|Trig|Acos(Degrees) (Math|Vector|DotProduct (Math|Vector|Normalize (Math|Vector|SelectVector (Math|Conversions|ToVector(Vector2D) (- (Math|Conversions|ToVector2D(Vector) (LocalVariables|GetRight%20Spline%20Location)) (Math|Conversions|ToVector2D(Vector) (LocalVariables|GetRight%20Sample%20Location)))) (LocalVariables|GetRight%20Spline%20Location) (LocalVariables|GetCalculate%202D%20Distance))) (LocalVariables|GetPlayer%20Forward%20Vector)))))
                          (Variables|Internal|SetSound%20Spline%20Data _structout_2))
                        (:then_1
                          (Variables|SoundSpline|GetDebug
                            (:then
                              (Rendering|Debug|DrawDebugSphere (LocalVariables|GetLeft%20Sample%20Location) 1e+01 12 "(R=0.000000,G=1.000000,B=1.000000,A=1.000000)" _time_until_next_update 2.0)
                              (Rendering|Debug|DrawDebugSphere (LocalVariables|GetRight%20Sample%20Location) 1e+01 12 "(R=1.000000,G=0.000000,B=1.000000,A=1.000000)" _time_until_next_update 2.0)
                              (Rendering|Debug|DrawDebugLine (LocalVariables|GetLeft%20Sample%20Location) _returnvalue_3 "(R=1.000000,G=0.125330,B=0.252830,A=1.000000)" _time_until_next_update 2.0)
                              (Rendering|Debug|DrawDebugLine (LocalVariables|GetRight%20Sample%20Location) _returnvalue_5 "(R=1.000000,G=0.777053,B=0.105510,A=1.000000)" _time_until_next_update 2.0)
                              (Rendering|Debug|DrawDebugSphere _returnvalue_3 25.0 12 "(R=0.000000,G=1.000000,B=0.000000,A=1.000000)" _time_until_next_update 2.0)
                              (Rendering|Debug|DrawDebugSphere _returnvalue_5 25.0 12 "(R=1.000000,G=0.000000,B=0.000000,A=1.000000)" _time_until_next_update 2.0))
                            (:else)))))))))))))))

(fn DoWork
  (SoundSpline|ComputeDistance)
  (SoundSpline|UpdateSplineAudio)
  (CallFunction|ScheduleUpdate (SoundSpline|GetTimeUntilNextUpdate)))

(fn GetSoundSplineData
  (return (Variables|Internal|GetSound%20Spline%20Data)))

(fn GetTimeUntilNextUpdate
  (bind (_tickinnerradius_31_ec99378b47c3005bc2c232889cd21f7a _tickfalloffdistance_33_6791e330486c65a6812b789d950c67b9 _mintickinterval_30_eac2223d45795382365f8d82f689bd5a _maxtickinterval_27_9e34ca984a83c90b6fa8c5a77a0fae9e) (Utilities|Struct|BreakSoundSplineSettings (Variables|SoundSpline|GetSound%20Spline%20Settings)))
  (bind (_leftdistance_21_c7542ba64fcc87b5473309909a599cfd _rightdistance_23_b0ca2147461d4421f95986bb44460265) (Utilities|Struct|BreakSoundSplineData (Variables|Internal|GetSound%20Spline%20Data)))
  (return (Math|Float|MapRangeClamped (Math|Float|Min(Float) _leftdistance_21_c7542ba64fcc87b5473309909a599cfd _rightdistance_23_b0ca2147461d4421f95986bb44460265) _tickinnerradius_31_ec99378b47c3005bc2c232889cd21f7a _tickfalloffdistance_33_6791e330486c65a6812b789d950c67b9 _mintickinterval_30_eac2223d45795382365f8d82f689bd5a _maxtickinterval_27_9e34ca984a83c90b6fa8c5a77a0fae9e)))

(fn SetSplineComponent
  (params (Spline%20Component :type Spline%20Component%20Object%20Reference))
  (SoundSpline|SetSplineComponentInternal Spline%20Component)
  (Development|PrintText (Utilities|Text|FormatText "NSLOCTEXT(\"[88B3029E41F5B6F0C32F3408DCEADECD]\", \"8364085B4EFD889F004F93B51DE002EA\", \"Registered spline {spline}\")" (Variables|GetSpline%20Component))))

(fn SetSplineComponentInternal
  (params (Spline%20Component :type Spline%20Component%20Object%20Reference))
  (Utilities|IsValid Spline%20Component
    (:Is%20Valid
      (bind _returnvalue (Game|AddComponentbyClass (Components|GetOwner) "/Script/Engine.AudioComponent" false (struct :Location "0, 0, 0" :Rotation "0, 0, 0" :Scale "1.000000,1.000000,1.000000")))
      (Variables|Internal|SetSound%20Spline%20Data (Utilities|Struct|MakeSoundSplineData (Variables|GetSpline%20Component) _returnvalue)))))

(fn StartSplineAudio
  (Audio|Components|Audio|SetSound (Utilities|Struct|BreakSoundSplineData (Variables|Internal|GetSound%20Spline%20Data)) (Utilities|Struct|BreakSoundSplineSettings (Variables|SoundSpline|GetSound%20Spline%20Settings)))
  (SoundSpline|UpdateSplineAudioParameters)
  (Audio|Components|Audio|FadeIn (Utilities|Struct|BreakSoundSplineData (Variables|Internal|GetSound%20Spline%20Data)) (Utilities|Struct|BreakSoundSplineSettings (Variables|SoundSpline|GetSound%20Spline%20Settings)) 1.0 0.0 "Linear" :!class /Script/Engine.AudioComponent))

(fn StopSplineAudio
  (Audio|Components|Audio|FadeOut (Utilities|Struct|BreakSoundSplineData (Variables|Internal|GetSound%20Spline%20Data)) (Utilities|Struct|BreakSoundSplineSettings (Variables|SoundSpline|GetSound%20Spline%20Settings)) 0.0 "Linear" :!class /Script/Engine.AudioComponent))

(fn UpdateSplineAudio
  (bind (_leftdistance_21_c7542ba64fcc87b5473309909a599cfd _rightdistance_23_b0ca2147461d4421f95986bb44460265) (Utilities|Struct|BreakSoundSplineData (Variables|Internal|GetSound%20Spline%20Data)))
  (Utilities|IsValid (Utilities|Struct|BreakSoundSplineData (Variables|Internal|GetSound%20Spline%20Data))
    (:Is%20Valid
      (bind _returnvalue (Fortnite|Utility|IsValueWithinRange (Math|Float|Min(Float) _leftdistance_21_c7542ba64fcc87b5473309909a599cfd _rightdistance_23_b0ca2147461d4421f95986bb44460265) (Utilities|Struct|MakeDoubleRange (struct :Type "Inclusive" :Value 0.0) (Utilities|Struct|MakeDoubleRangeBound "Inclusive" (Utilities|Struct|BreakSoundSplineSettings (Variables|SoundSpline|GetSound%20Spline%20Settings))))))
      (if _returnvalue
        (if (Audio|Components|Audio|IsPlaying (Utilities|Struct|BreakSoundSplineData (Variables|Internal|GetSound%20Spline%20Data)))
          (SoundSpline|UpdateSplineAudioParameters)
          (else
            (SoundSpline|StartSplineAudio)))
        (else
          (SoundSpline|StopSplineAudio))))))

(fn UpdateSplineAudioParameters
  (bind _sound_spline_data (Variables|Internal|GetSound%20Spline%20Data))
  (bind (_audiocomponent_8_3ee2df774fcf2e32668b5b9b7aa0791d _leftdistance_21_c7542ba64fcc87b5473309909a599cfd _leftangle_35_c6ad39a143b5c6be47d6a5a35772e682) (Utilities|Struct|BreakSoundSplineData _sound_spline_data))
  (bind _sound_spline_data_2 (Variables|Internal|GetSound%20Spline%20Data))
  (bind (_audiocomponent_8_3ee2df774fcf2e32668b5b9b7aa0791d_2 _rightdistance_23_b0ca2147461d4421f95986bb44460265 _rightangle_36_ee3eca9648194df9345d39b5198bf212) (Utilities|Struct|BreakSoundSplineData _sound_spline_data_2))
  (Utilities|FlowControl|Sequence
    (:then_0
      (Audio|Parameter|SetFloatParameter _audiocomponent_8_3ee2df774fcf2e32668b5b9b7aa0791d "SoundSpline.Distance.Left" _leftdistance_21_c7542ba64fcc87b5473309909a599cfd :!class /Script/AudioExtensions.AudioParameterControllerInterface)
      (Audio|Parameter|SetFloatParameter _audiocomponent_8_3ee2df774fcf2e32668b5b9b7aa0791d "SoundSpline.Angle.Left" _leftangle_35_c6ad39a143b5c6be47d6a5a35772e682 :!class /Script/AudioExtensions.AudioParameterControllerInterface))
    (:then_1
      (Utilities|FlowControl|Sequence
        (:then_0
          (Audio|Parameter|SetFloatParameter _audiocomponent_8_3ee2df774fcf2e32668b5b9b7aa0791d_2 "SoundSpline.Distance.Right" _rightdistance_23_b0ca2147461d4421f95986bb44460265 :!class /Script/AudioExtensions.AudioParameterControllerInterface)
          (Audio|Parameter|SetFloatParameter _audiocomponent_8_3ee2df774fcf2e32668b5b9b7aa0791d_2 "SoundSpline.Angle.Right" _rightangle_36_ee3eca9648194df9345d39b5198bf212 :!class /Script/AudioExtensions.AudioParameterControllerInterface))
        (:then_1
          (Utilities|FlowControl|Sequence
            (:then_0
              (Audio|Parameter|SetFloatParameter _audiocomponent_8_3ee2df774fcf2e32668b5b9b7aa0791d_2 "SoundSpline.SampleWidth" (Utilities|Struct|BreakSoundSplineSettings (Variables|SoundSpline|GetSound%20Spline%20Settings)) :!class /Script/AudioExtensions.AudioParameterControllerInterface))))))))
